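# Physics asset for the VitalInventor male medium player body cosmetic, covering a map and four belt items hung from the pelvis.
# Each item pairs a base anchor bone with a tiny capsule, angularly free to the pelvis, and a fitted dyn capsule constrained to that base.
# Narrow per-item swing and twist limits restrain the dangling pieces, while pelvis and left thigh capsules are the only body shapes.
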PhysicsAsset /BRCosmetics/Characters/Player/Male/Medium/Bodies/M_MED_VitalInventor/Meshes/M_MED_VitalInventor_Physics
bodies {
 base_beltItem_01 [{shapeType=Capsule, length=0.505, radius=0.505, shapeName=base_beltItem_01_capsule, rotation=<rot 0 -90 0>}]
 base_beltItem_02 [{shapeType=Capsule, length=0.505, radius=0.505, shapeName=base_beltItem_02_capsule, rotation=<rot 0 -90 0>}]
 base_beltItem_03 [{shapeType=Capsule, length=0.505, radius=0.505, shapeName=base_beltItem_03_capsule, rotation=<rot 0 -90 0>}]
 base_beltitem_04 [{shapeType=Capsule, length=0.505, radius=0.505, shapeName=base_beltitem_04_capsule, rotation=<rot 0 -90 0>}]
 base_map [{shapeType=Capsule, length=0.505, radius=0.505, shapeName=base_map_capsule, rotation=<rot 0 -90 0>}]
 dyn_beltItem_01 [{
   shapeType=Capsule
   length=4.7743
   radius=2.8122
   shapeName=dyn_beltItem_01_capsule
   center=<2.6516 1.4311 -6.5837>
   rotation=<rot -4.9953 -170.5383 -177.4873>
  }]
 dyn_beltItem_02 [{
   shapeType=Capsule
   length=7.1188
   radius=3.6638
   shapeName=dyn_beltItem_02_capsule
   center=<1.4452 -0.6014 -4.7528>
   rotation=<rot 4.5514 -90 -0.1209>
  }]
 dyn_beltItem_03 [{
   shapeType=Capsule
   length=5.3929
   radius=2.2546
   shapeName=dyn_beltItem_03_capsule
   center=<1.1449 -2.4477 -3.4704>
   rotation=<rot 8.1432 -90 -1.5751>
  }]
 dyn_beltitem_04 [{
   shapeType=Capsule
   length=3.638
   radius=4.2514
   shapeName=dyn_beltitem_04_capsule
   center=<4.1412 0.0163 -5.2744>
   rotation=<rot 6.5384 -90 -8.1484>
  }]
 dyn_map [{shapeType=Capsule, length=27.2477, radius=4.7687, shapeName=dyn_map_capsule, center=<3.5 0 0>, rotation=<rot -0.0003 0 -70>}]
 pelvis [{shapeType=Capsule, length=13.476, radius=9.2492, center=<0 -1.8546 0>}]
 thigh_l [{shapeType=Capsule, length=23.3222, radius=7.8712, shapeName=thigh_l_capsule, center=<-14.6671 0 0>, rotation=<rot 90 -179.156 -179.156>}]
}
constraints [
 {
  bone1Name=base_map
  bone2Name=pelvis
  swing1LimitDegrees=45
  swing1Motion=Free
  swing2LimitDegrees=45
  swing2Motion=Free
  twistLimitDegrees=45
  twistMotion=Free
 }
 {bone1Name=dyn_map, bone2Name=base_map, swing1LimitDegrees=10, swing2LimitDegrees=40, twistLimitDegrees=5}
 {
  bone1Name=thigh_l
  bone2Name=pelvis
  swing1LimitDegrees=45
  swing1Motion=Free
  swing2LimitDegrees=45
  swing2Motion=Free
  twistLimitDegrees=45
  twistMotion=Free
 }
 {
  bone1Name=base_beltItem_01
  bone2Name=pelvis
  swing1LimitDegrees=45
  swing1Motion=Free
  swing2LimitDegrees=45
  swing2Motion=Free
  twistLimitDegrees=45
  twistMotion=Free
 }
 {bone1Name=dyn_beltItem_01, bone2Name=base_beltItem_01, swing1LimitDegrees=5, swing2LimitDegrees=5, twistLimitDegrees=5}
 {
  bone1Name=base_beltItem_02
  bone2Name=pelvis
  swing1LimitDegrees=45
  swing1Motion=Free
  swing2LimitDegrees=45
  swing2Motion=Free
  twistLimitDegrees=45
  twistMotion=Free
 }
 {
  bone1Name=base_beltItem_03
  bone2Name=pelvis
  swing1LimitDegrees=45
  swing1Motion=Free
  swing2LimitDegrees=45
  swing2Motion=Free
  twistLimitDegrees=45
  twistMotion=Free
 }
 {bone1Name=dyn_beltItem_02, bone2Name=base_beltItem_02, swing1LimitDegrees=20, swing2LimitDegrees=10, twistLimitDegrees=20}
 {bone1Name=dyn_beltItem_03, bone2Name=base_beltItem_03, swing1LimitDegrees=5, swing2LimitDegrees=10, twistLimitDegrees=5}
 {bone1Name=dyn_beltitem_04, bone2Name=base_beltitem_04, swing1LimitDegrees=5, swing2LimitDegrees=10, twistLimitDegrees=5}
 {
  bone1Name=base_beltitem_04
  bone2Name=pelvis
  swing1LimitDegrees=45
  swing1Motion=Free
  swing2LimitDegrees=45
  swing2Motion=Free
  twistLimitDegrees=45
  twistMotion=Free
 }
]
