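# Post-processes the M_MED_VitalInventor cosmetic body, following the attached parent mesh's pose and adding secondary motion to its accessories.
# Weighted multi-bone constraints anchor belt items and map to hips, spine and thighs, and the left knot to the calf, before a deform Control Rig.
# AnimDynamics then swings the knot, and a pelvis-space RigidBody node simulates the physics asset.
# Each update, a deprecated gravity-override function sets that simulation's gravity, with tuned skydive float parameters.
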AnimBlueprint /BRCosmetics/Characters/Player/Male/Medium/Bodies/M_MED_VitalInventor/Meshes/M_MED_VitalInventor_AnimBP
target_skeleton=&/M_MED_VitalInventor_Skeleton
graphs [
 AnimGraph {nodes [
   n1 = "Constraint - base_beltItem_01 (PPPPPPPP)"!(
    ComponentPose=Local To Component(LocalPose=Copy Pose From Mesh() {bUseAttachedParent=true})
    Alpha=1
    ConstraintWeights_0=0.68
    ConstraintWeights_1:float=0.1066
    ConstraintWeights_2:float=0.0656
    ConstraintWeights_3:float=0.0578
    ConstraintWeights_4:float=0.0493
    ConstraintWeights_5:float=0.0325
    ConstraintWeights_6:float=0.007
    ConstraintWeights_7:float=0.0011
   ) -> (Pose) {
    alphaBoolBlend {blendOption=Linear}
    boneToModify {boneName=base_beltItem_01}
    constraintSetup [
     {targetBone {boneName=pelvis}}
     {targetBone {boneName=deform_groin_r}}
     {targetBone {boneName=thigh_twist_01_r}}
     {targetBone {boneName=spine_02}}
     {targetBone {boneName=spine_01}}
     {targetBone {boneName=spine_03}}
     {targetBone {boneName=deform_glute_r}}
     {targetBone {boneName=thigh_twist_01_l}}
    ]
    constraintWeights [1,1,1,1,1,1,1,1]
   }
   n2 = "Constraint - base_beltItem_02 (PPPPPPPP)"!(
    ComponentPose=@n1
    Alpha=1
    ConstraintWeights_0=0.6584
    ConstraintWeights_1:float=0.2137
    ConstraintWeights_2:float=0.0481
    ConstraintWeights_3:float=0.0302
    ConstraintWeights_4:float=0.0258
    ConstraintWeights_5:float=0.0172
    ConstraintWeights_6:float=0.0039
    ConstraintWeights_7:float=0.0027
   ) -> (Pose) {
    alphaBoolBlend {blendOption=Linear}
    boneToModify {boneName=base_beltItem_02}
    constraintSetup [
     {targetBone {boneName=pelvis}}
     {targetBone {boneName=deform_groin_l}}
     {targetBone {boneName=thigh_twist_01_l}}
     {targetBone {boneName=spine_02}}
     {targetBone {boneName=spine_01}}
     {targetBone {boneName=spine_03}}
     {targetBone {boneName=thigh_twist_01_r}}
     {targetBone {boneName=deform_groin_r}}
    ]
    constraintWeights [1,1,1,1,1,1,1,1]
   }
   n3 = "Constraint - base_beltItem_03 (PPPPPPPP)"!(
    ComponentPose=@n2
    Alpha=1
    ConstraintWeights_0=0.6502
    ConstraintWeights_1:float=0.1894
    ConstraintWeights_2:float=0.0567
    ConstraintWeights_3:float=0.0408
    ConstraintWeights_4:float=0.0349
    ConstraintWeights_5:float=0.0231
    ConstraintWeights_6:float=0.0024
    ConstraintWeights_7:float=0.0024
   ) -> (Pose) {
    alphaBoolBlend {blendOption=Linear}
    boneToModify {boneName=base_beltItem_03}
    constraintSetup [
     {targetBone {boneName=pelvis}}
     {targetBone {boneName=deform_groin_l}}
     {targetBone {boneName=thigh_twist_01_l}}
     {targetBone {boneName=spine_02}}
     {targetBone {boneName=spine_01}}
     {targetBone {boneName=spine_03}}
     {targetBone {boneName=thigh_twist_01_r}}
     {targetBone {boneName=deform_glute_l}}
    ]
    constraintWeights [1,1,1,1,1,1,1,1]
   }
   n4 = "Constraint - base_beltitem_04 (PPPPPPPP)"!(
    ComponentPose=@n3
    Alpha=1
    ConstraintWeights_0=0.7728
    ConstraintWeights_1:float=0.086
    ConstraintWeights_2:float=0.0734
    ConstraintWeights_3:float=0.049
    ConstraintWeights_4:float=0.0098
    ConstraintWeights_5:float=0.0055
    ConstraintWeights_6:float=0.0035
    ConstraintWeights_7:float=0.0001
   ) -> (Pose) {
    alphaBoolBlend {blendOption=Linear}
    boneToModify {boneName=base_beltitem_04}
    constraintSetup [
     {targetBone {boneName=pelvis}}
     {targetBone {boneName=spine_02}}
     {targetBone {boneName=spine_01}}
     {targetBone {boneName=spine_03}}
     {targetBone {boneName=thigh_twist_01_l}}
     {targetBone {boneName=deform_glute_l}}
     {targetBone {boneName=deform_groin_l}}
     {targetBone {boneName=spine_04}}
    ]
    constraintWeights [1,1,1,1,1,1,1,1]
   }
   n5 = "Constraint - base_map (PPPPPPPP)"!(
    ComponentPose=@n4
    Alpha=1
    ConstraintWeights_0=0.9561
    ConstraintWeights_1:float=0.0109
    ConstraintWeights_2:float=0.0097
    ConstraintWeights_3:float=0.0053
    ConstraintWeights_4:float=0.0053
    ConstraintWeights_5:float=0.005
    ConstraintWeights_6:float=0.0038
    ConstraintWeights_7:float=0.0038
   ) -> (Pose) {
    alphaBoolBlend {blendOption=Linear}
    boneToModify {boneName=base_map}
    constraintSetup [
     {targetBone {boneName=pelvis}}
     {targetBone {boneName=spine_01}}
     {targetBone {boneName=spine_02}}
     {targetBone {boneName=deform_glute_r}}
     {targetBone {boneName=deform_glute_l}}
     {targetBone {boneName=spine_03}}
     {targetBone {boneName=thigh_twist_01_l}}
     {targetBone {boneName=thigh_twist_01_r}}
    ]
    constraintWeights [1,1,1,1,1,1,1,1]
   }
   n6 = "Constraint - base_knot_Lf (PPP)"!(ComponentPose=@n5, Alpha=1, ConstraintWeights_0=0.6923, ConstraintWeights_1:float=0.2772, ConstraintWeights_2:float=0.0305) -> (Pose) {
    alphaBoolBlend {blendOption=Linear}
    boneToModify {boneName=base_knot_Lf}
    constraintSetup [{targetBone {boneName=calf_twist_02_l}}, {targetBone {boneName=calf_twist_01_l}}, {targetBone {boneName=calf_l}}]
    constraintWeights [1,1,1]
   }
   n7 = Control Rig(Source=Component To Local(ComponentPose=@n6)) {
    alphaBoolBlend {blendOption=Linear}
    controlRigAssetReference {blueprintRigClass=&/M_MED_VitalInventor_DeformRig_Master.M_MED_VitalInventor_DeformRig_Master_C}
   }
   n8 = Anim Dynamics - Bone: dyn_knot_01_Lf(ComponentPose=Local To Component(LocalPose=@n7), Alpha=1, ExternalForce="0,0,0") {
    angularSpringConstant=200
    simulationSpace=RootRelative
    alphaBoolBlend {blendOption=Linear}
    boundBone {boneName=dyn_knot_01_Lf}
    componentLinearAccScale=<0.2 0.2 0.2>
    componentLinearVelScale=<0.2 0.2 0.2>
    physicsBodyDefinitions [{
      boundBone {boneName=dyn_knot_01_Lf}
      boxExtents=<20 20 20>
      constraintSetup {angularLimitsMax=<0 5 5>, angularLimitsMin=<0 0 -5>, angularTarget=<0 0 0>}
      localJointOffset=<0.5 0 0>
     }]
   }
   n9 = Anim Dynamics - Bone: dyn_knot_02_Lf(ComponentPose=@n8, Alpha=1, ExternalForce="0,0,0") {
    angularSpringConstant=200
    simulationSpace=RootRelative
    alphaBoolBlend {blendOption=Linear}
    boundBone {boneName=dyn_knot_02_Lf}
    componentLinearAccScale=<0.2 0.2 0.2>
    componentLinearVelScale=<0.2 0.2 0.2>
    physicsBodyDefinitions [{
      boundBone {boneName=dyn_knot_02_Lf}
      boxExtents=<17 17 17>
      constraintSetup {angularLimitsMax=<0 10 20>, angularLimitsMin=<0 0 -20>, angularTarget=<0 0 0>}
      localJointOffset=<1 0 0>
     }]
   }
   n10 = RigidBody!(ComponentPose=@n9, Alpha, OverrideWorldGravity:Vector=Get GravityOverride(), ExternalForce) -> (Pose) {
    bClampLinearTranslationLimitToRefPose=true
    bOverrideWorldGravity=true
    overridePhysicsAsset=&/M_MED_VitalInventor_Physics
    simulationSpace=BaseBoneSpace
    alphaBoolBlend {blendOption=Linear}
    baseBoneRef {boneName=pelvis}
    componentLinearAccScale=<0.4 0.4 0.4>
    componentLinearVelScale=<0.4 0.4 0.4>
    overrideWorldGravity {z=-980}
    simSpaceSettings {
     maxAngularAcceleration=1000
     maxAngularVelocity=3
     maxLinearAcceleration=700
     maxLinearVelocity=500
     worldAlpha=0.25
     externalLinearDragV=<0.2 0.2>
    }
   }
   Output Pose(Result=Component To Local(ComponentPose=@n10)) {layerGroup=None}
  ]}
 EventGraph {nodes [
   Event BlueprintUpdateAnimation()
   > n1 = Gravity Override State Function DEPRECATED(CharacterPartAnimInstance=Self-Reference(), GameTimeInSeconds=GetGameTimeInSeconds(), GravityOverrideParams=Get Gravity Override Params())
   > Set GravityOverride(GravityOverride=@n1)
  ]}
]
variables [
 Gravity Override Params {type=GravityOverrideParamsStruct, default {skydive_FloatingDown {gravityOverride=<0 -700 200>, windAmplitude=<200 -500 600>}}}
 GravityOverride {type=Vector}
]
