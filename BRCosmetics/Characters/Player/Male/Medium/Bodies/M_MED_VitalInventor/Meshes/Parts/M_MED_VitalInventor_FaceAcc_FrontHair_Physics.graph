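# Physics asset for the front hair of the male medium VitalInventor character, covering a left lock, a right lock and a bang.
# Each is a three-bone chain of capsules hung from a base bone on the head, with tight swing limits and mostly zero twist.
# Capsules on the head, neck, upper spine and clavicles form the body volume around the hair, linked by free-rotation constraints, as are the hair bases.
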PhysicsAsset /BRCosmetics/Characters/Player/Male/Medium/Bodies/M_MED_VitalInventor/Meshes/Parts/M_MED_VitalInventor_FaceAcc_FrontHair_Physics
bodies {
 base_bang_l [{shapeType=Capsule, length=0.505, radius=0.505, shapeName=base_bang_l_capsule, rotation=<rot 0 -90 0>}]
 base_hair_front_l [{shapeType=Capsule, length=0.505, radius=0.505, shapeName=base_hair_front_l_capsule, rotation=<rot 0 -90 0>}]
 base_hair_front_r [{shapeType=Capsule, length=0.505, radius=0.505, shapeName=base_hair_front_r_capsule, rotation=<rot 0 -90 0>}]
 clavicle_l [
  {shapeType=Capsule, length=16.4579, radius=6.8463, shapeName=clavicle_l_capsule, center=<9.3871 0 1.5881>, rotation=<rot -0.0004 -90 110>}
  {shapeType=Capsule, length=18.1026, radius=3.4356, shapeName=clavicle_l_capsule, center=<9.1813 2.4555 11.4498>, rotation=<rot 0 0 84.9993>}
 ]
 clavicle_r [{shapeType=Capsule, length=16.4579, radius=6.8463, shapeName=clavicle_l_capsule, center=<-9.3871 0 -1.5881>, rotation=<rot -0.0004 -90 -70>}]
 dyn_bang_01 [{shapeType=Capsule, length=3, radius=1.5, shapeName=dyn_bang_01_capsule, center=<1.7493 0 0>, rotation=<rot 90 0 0>}]
 dyn_bang_02 [{shapeType=Capsule, length=2.6767, radius=1.8233, shapeName=dyn_bang_02_capsule, center=<3.1096 0 0>, rotation=<rot 90 0 0>}]
 dyn_bang_03 [{shapeType=Capsule, length=3, radius=1.5, shapeName=dyn_bang_03_capsule, center=<0.9727 -1.1886 0>, rotation=<rot 90 0 0>}]
 dyn_hair_front_01_l [{shapeType=Capsule, length=2, radius=2, shapeName=dyn_hair_front_01_l_capsule, center=<1 0 0>, rotation=<rot 90 0 0>}]
 dyn_hair_front_01_r [{shapeType=Capsule, length=2, radius=2, shapeName=dyn_hair_front_01_r_capsule, center=<2.1458 0 0>, rotation=<rot 90 0 0>}]
 dyn_hair_front_02_l [{
   shapeType=Capsule
   length=4.5548
   radius=2.0518
   shapeName=dyn_hair_front_02_l_capsule
   center=<2.0796 0 -2.3748>
   rotation=<rot 59.9997 -179.9999 -180>
  }]
 dyn_hair_front_02_r [{shapeType=Capsule, length=1.9174, radius=2.0826, shapeName=dyn_hair_front_02_r_capsule, center=<2.0174 0 -1.3867>, rotation=<rot 90 0 0>}]
 dyn_hair_front_03_l [{shapeType=Capsule, length=3.644, radius=1.4792, shapeName=dyn_hair_front_03_l_capsule, center=<2.7473 0.7032 0>, rotation=<rot 90 0 0>}]
 dyn_hair_front_03_r [{shapeType=Capsule, length=1.2174, radius=2.7826, shapeName=dyn_hair_front_03_r_capsule, center=<1 0 -0.2366>, rotation=<rot 90 0 0>}]
 head [{shapeType=Capsule, length=6.5846, radius=6.5381, center=<2.8052 3.066 0.3052>, rotation=<rot 90 180 -180>}]
 neck_01 [{shapeType=Capsule, length=8.3668, radius=3.7081, shapeName=neck_01_capsule, center=<7.7479 0 0>, rotation=<rot 0 -90.0002 90>}]
 spine_05 [{shapeType=Capsule, length=21.0262, radius=10.0547, shapeName=spine_05_capsule, center=<5.0916 -3.1095 0>, rotation=<rot 0 -90 0>}]
}
constraints [
 {
  bone1Name=base_hair_front_l
  bone2Name=head
  swing1LimitDegrees=45
  swing1Motion=Free
  swing2LimitDegrees=45
  swing2Motion=Free
  twistLimitDegrees=45
  twistMotion=Free
 }
 {bone1Name=dyn_hair_front_01_l, bone2Name=base_hair_front_l, swing1LimitDegrees=15, swing2LimitDegrees=30, twistLimitDegrees=0}
 {bone1Name=dyn_hair_front_02_l, bone2Name=dyn_hair_front_01_l, swing1LimitDegrees=15, swing2LimitDegrees=20, twistLimitDegrees=0}
 {bone1Name=dyn_hair_front_03_l, bone2Name=dyn_hair_front_02_l, swing1LimitDegrees=10, swing2LimitDegrees=20, twistLimitDegrees=0}
 {
  bone1Name=base_hair_front_r
  bone2Name=head
  swing1LimitDegrees=45
  swing1Motion=Free
  swing2LimitDegrees=45
  swing2Motion=Free
  twistLimitDegrees=45
  twistMotion=Free
 }
 {bone1Name=dyn_hair_front_01_r, bone2Name=base_hair_front_r, swing1LimitDegrees=15, swing2LimitDegrees=5, twistLimitDegrees=0}
 {bone1Name=dyn_hair_front_02_r, bone2Name=dyn_hair_front_01_r, swing1LimitDegrees=10, swing2LimitDegrees=20, twistLimitDegrees=20}
 {bone1Name=dyn_hair_front_03_r, bone2Name=dyn_hair_front_02_r, swing1LimitDegrees=10, swing2LimitDegrees=20, twistLimitDegrees=0}
 {
  bone1Name=head
  bone2Name=neck_01
  swing1LimitDegrees=45
  swing1Motion=Free
  swing2LimitDegrees=45
  swing2Motion=Free
  twistLimitDegrees=45
  twistMotion=Free
 }
 {
  bone1Name=neck_01
  bone2Name=spine_05
  swing1LimitDegrees=45
  swing1Motion=Free
  swing2LimitDegrees=45
  swing2Motion=Free
  twistLimitDegrees=45
  twistMotion=Free
 }
 {
  bone1Name=clavicle_l
  bone2Name=spine_05
  swing1LimitDegrees=45
  swing1Motion=Free
  swing2LimitDegrees=45
  swing2Motion=Free
  twistLimitDegrees=45
  twistMotion=Free
 }
 {
  bone1Name=clavicle_r
  bone2Name=spine_05
  swing1LimitDegrees=45
  swing1Motion=Free
  swing2LimitDegrees=45
  swing2Motion=Free
  twistLimitDegrees=45
  twistMotion=Free
 }
 {
  bone1Name=base_bang_l
  bone2Name=head
  swing1LimitDegrees=45
  swing1Motion=Free
  swing2LimitDegrees=45
  swing2Motion=Free
  twistLimitDegrees=45
  twistMotion=Free
 }
 {bone1Name=dyn_bang_02, bone2Name=dyn_bang_01, swing1LimitDegrees=15, swing2LimitDegrees=10, twistLimitDegrees=0}
 {bone1Name=dyn_bang_01, bone2Name=base_bang_l, swing1LimitDegrees=10, swing2LimitDegrees=10, twistLimitDegrees=0}
 {bone1Name=dyn_bang_03, bone2Name=dyn_bang_02, swing1LimitDegrees=15, swing2LimitDegrees=10, twistLimitDegrees=0}
]
