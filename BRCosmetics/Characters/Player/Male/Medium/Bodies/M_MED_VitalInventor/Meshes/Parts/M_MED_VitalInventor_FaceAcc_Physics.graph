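# Physics asset for the VitalInventor medium male face accessory mesh, set up for secondary motion of back hair and a tail hanging from the head.
# Three four-link hair chains (center, left, right) and a three-link tail start at base bones joined to the head by free-rotation constraints.
# Capsules on head, neck, upper spine and clavicles form the body volume around them.
# Chain joints have tight, asymmetric swing limits, generally loosening toward hair tips, with little or no twist.
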PhysicsAsset /BRCosmetics/Characters/Player/Male/Medium/Bodies/M_MED_VitalInventor/Meshes/Parts/M_MED_VitalInventor_FaceAcc_Physics
bodies {
 base_hair_back [{shapeType=Capsule, length=0.505, radius=0.505, shapeName=base_hair_back_capsule, rotation=<rot 0 -90 0>}]
 base_hair_back_l [{shapeType=Capsule, length=0.505, radius=0.505, shapeName=base_hair_back_l_capsule, rotation=<rot 0 -90 0>}]
 base_hair_back_r [{shapeType=Capsule, length=0.505, radius=0.505, shapeName=base_hair_back_r_capsule, rotation=<rot 0 -90 0>}]
 base_tail [{shapeType=Capsule, length=0.505, radius=0.505, shapeName=base_tail_capsule, rotation=<rot 0 -90 0>}]
 clavicle_l [{
   shapeType=Capsule
   length=4.9677
   radius=11.0865
   shapeName=clavicle_l_capsule
   center=<11.5808 -0.4055 0.3325>
   rotation=<rot -0.0006 -90 110>
  }]
 clavicle_r [{
   shapeType=Capsule
   length=2.8422
   radius=9.1699
   shapeName=clavicle_l_capsule
   center=<-12.3781 -0.2778 1.461>
   rotation=<rot -2.1328 -80.2275 -77.8134>
  }]
 dyn_hair_back_01 [{shapeType=Capsule, length=2, radius=2, shapeName=dyn_hair_back_01_capsule, center=<5 0 0>, rotation=<rot 90 0 0>}]
 dyn_hair_back_01_l [{shapeType=Capsule, length=2, radius=2, shapeName=dyn_hair_back_01_l_capsule, center=<5 0 0>, rotation=<rot 90 0 0>}]
 dyn_hair_back_01_r [{shapeType=Capsule, length=2, radius=2, shapeName=dyn_hair_back_01_r_capsule, center=<5 0 0>, rotation=<rot 90 0 0>}]
 dyn_hair_back_02 [{shapeType=Capsule, length=0.75, radius=2.75, shapeName=dyn_hair_back_02_capsule, center=<6.1 0 -0.5>, rotation=<rot 80.0005 -180 -180>}]
 dyn_hair_back_02_l [{shapeType=Capsule, length=2.25, radius=2.75, shapeName=dyn_hair_back_02_l_capsule, center=<6.1 0 -1.1274>, rotation=<rot 90 0 0>}]
 dyn_hair_back_02_r [{
   shapeType=Capsule
   length=2.5
   radius=2.75
   shapeName=dyn_hair_back_02_r_capsule
   center=<6.1395 0.0609 -0.7238>
   rotation=<rot 72.0365 -122.9443 145.736>
  }]
 dyn_hair_back_03 [{shapeType=Capsule, length=3, radius=2.5, shapeName=dyn_hair_back_03_capsule, center=<4 0 0>, rotation=<rot 85 -180 -180>}]
 dyn_hair_back_03_l [{shapeType=Capsule, length=3, radius=2.5, shapeName=dyn_hair_back_03_l_capsule, center=<4 0 -1>, rotation=<rot 90 0 0>}]
 dyn_hair_back_03_r [{shapeType=Capsule, length=3, radius=2.75, shapeName=dyn_hair_back_03_r_capsule, center=<4 -2.4684 1.2253>, rotation=<rot 75.0001 -180 -180>}]
 dyn_hair_back_04 [{shapeType=Capsule, length=1.75, radius=2.75, shapeName=dyn_hair_back_04_capsule, center=<4.5 0 0.8427>, rotation=<rot 90 0 0>}]
 dyn_hair_back_04_l [{shapeType=Capsule, length=3, radius=2, shapeName=dyn_hair_back_04_l_capsule, center=<4.5 1.1992 -0.5>, rotation=<rot 90 0 0>}]
 dyn_hair_back_04_r [{shapeType=Capsule, length=2.25, radius=2.75, shapeName=dyn_hair_back_04_r_capsule, center=<4.5 -1.0935 1.3664>, rotation=<rot 90 0 0>}]
 dyn_tail_01 [{shapeType=Capsule, length=3.829, radius=3.9142, shapeName=dyn_tail_01_capsule, center=<4 0 0>, rotation=<rot 90 0 0>}]
 dyn_tail_02 [{shapeType=Capsule, length=0.2958, radius=5.504, shapeName=dyn_tail_02_capsule, center=<5.8475 0 -1.193>, rotation=<rot 90 0 0>}]
 dyn_tail_03 [{shapeType=Capsule, length=2.0281, radius=3.9221, shapeName=dyn_tail_03_capsule, center=<5 0 -1>, rotation=<rot 90 0 0>}]
 head [{shapeType=Capsule, length=6.3133, radius=6.8094, center=<2.8052 0.668 0>, rotation=<rot 90 180 -180>}]
 neck_01 [{shapeType=Capsule, length=6.0223, radius=6.0525, shapeName=neck_01_capsule, center=<7.7479 0 0>, rotation=<rot 0 -90.0002 90>}]
 spine_04 [{shapeType=Capsule, length=16.255, radius=14.005, shapeName=spine_04_capsule, center=<11.4692 0 0>, rotation=<rot 0 -90 0>}]
 spine_05 [
  {shapeType=Capsule, length=15.6113, radius=11.4696, shapeName=spine_05_capsule, center=<5.2388 0.7859 0>, rotation=<rot 0 -90 0>}
  {shapeType=Capsule, length=27.9678, radius=15, shapeName=spine_05_capsule, center=<14.7813 22.7049 0>}
 ]
}
constraints [
 {
  bone1Name=base_hair_back
  bone2Name=head
  swing1LimitDegrees=45
  swing1Motion=Free
  swing2LimitDegrees=45
  swing2Motion=Free
  twistLimitDegrees=45
  twistMotion=Free
 }
 {bone1Name=dyn_hair_back_01, bone2Name=base_hair_back, swing1LimitDegrees=1, swing2LimitDegrees=15, twistLimitDegrees=0}
 {bone1Name=dyn_hair_back_02, bone2Name=dyn_hair_back_01, swing1LimitDegrees=10, swing2LimitDegrees=20, twistLimitDegrees=0}
 {bone1Name=dyn_hair_back_03, bone2Name=dyn_hair_back_02, swing1LimitDegrees=10, swing2LimitDegrees=40, twistLimitDegrees=0}
 {
  bone1Name=dyn_hair_back_04
  bone2Name=dyn_hair_back_03
  swing1LimitDegrees=20
  swing2LimitDegrees=50
  twistLimitDegrees=0
  twistMotion=Locked
 }
 {
  bone1Name=base_hair_back_l
  bone2Name=head
  swing1LimitDegrees=45
  swing1Motion=Free
  swing2LimitDegrees=45
  swing2Motion=Free
  twistLimitDegrees=45
  twistMotion=Free
 }
 {bone1Name=dyn_hair_back_01_l, bone2Name=base_hair_back_l, swing1LimitDegrees=1, swing2LimitDegrees=15, twistLimitDegrees=0}
 {bone1Name=dyn_hair_back_02_l, bone2Name=dyn_hair_back_01_l, swing1LimitDegrees=5, swing2LimitDegrees=20, twistLimitDegrees=0}
 {bone1Name=dyn_hair_back_03_l, bone2Name=dyn_hair_back_02_l, swing1LimitDegrees=10, swing2LimitDegrees=40, twistLimitDegrees=0}
 {
  bone1Name=dyn_hair_back_04_l
  bone2Name=dyn_hair_back_03_l
  swing1LimitDegrees=10
  swing2LimitDegrees=50
  twistLimitDegrees=0
  twistMotion=Locked
 }
 {
  bone1Name=base_hair_back_r
  bone2Name=head
  swing1LimitDegrees=45
  swing1Motion=Free
  swing2LimitDegrees=45
  swing2Motion=Free
  twistLimitDegrees=45
  twistMotion=Free
 }
 {bone1Name=dyn_hair_back_01_r, bone2Name=base_hair_back_r, swing1LimitDegrees=0, swing2LimitDegrees=15, twistLimitDegrees=0}
 {
  bone1Name=dyn_hair_back_02_r
  bone2Name=dyn_hair_back_01_r
  swing1LimitDegrees=5
  swing2LimitDegrees=20
  twistLimitDegrees=0
  twistMotion=Locked
 }
 {bone1Name=dyn_hair_back_03_r, bone2Name=dyn_hair_back_02_r, swing1LimitDegrees=15, swing2LimitDegrees=40, twistLimitDegrees=0}
 {
  bone1Name=dyn_hair_back_04_r
  bone2Name=dyn_hair_back_03_r
  swing1LimitDegrees=15
  swing2LimitDegrees=50
  twistLimitDegrees=0
  twistMotion=Locked
 }
 {
  bone1Name=head
  bone2Name=neck_01
  swing1LimitDegrees=45
  swing1Motion=Free
  swing2LimitDegrees=45
  swing2Motion=Free
  twistLimitDegrees=45
  twistMotion=Free
 }
 {
  bone1Name=neck_01
  bone2Name=spine_05
  swing1LimitDegrees=45
  swing1Motion=Free
  swing2LimitDegrees=45
  swing2Motion=Free
  twistLimitDegrees=45
  twistMotion=Free
 }
 {
  bone1Name=clavicle_l
  bone2Name=spine_05
  swing1LimitDegrees=45
  swing1Motion=Free
  swing2LimitDegrees=45
  swing2Motion=Free
  twistLimitDegrees=45
  twistMotion=Free
 }
 {
  bone1Name=clavicle_r
  bone2Name=spine_05
  swing1LimitDegrees=45
  swing1Motion=Free
  swing2LimitDegrees=45
  swing2Motion=Free
  twistLimitDegrees=45
  twistMotion=Free
 }
 {
  bone1Name=base_tail
  bone2Name=head
  swing1LimitDegrees=45
  swing1Motion=Free
  swing2LimitDegrees=45
  swing2Motion=Free
  twistLimitDegrees=45
  twistMotion=Free
 }
 {bone1Name=dyn_tail_01, bone2Name=base_tail, swing1LimitDegrees=5, swing2LimitDegrees=30, twistLimitDegrees=45, twistMotion=Locked}
 {bone1Name=dyn_tail_02, bone2Name=dyn_tail_01, swing1LimitDegrees=20, swing2LimitDegrees=40, twistLimitDegrees=0}
 {bone1Name=dyn_tail_03, bone2Name=dyn_tail_02, swing1LimitDegrees=20, swing2LimitDegrees=20, twistLimitDegrees=0, twistMotion=Locked}
 {
  bone1Name=spine_05
  bone2Name=spine_04
  swing1LimitDegrees=45
  swing1Motion=Free
  swing2LimitDegrees=45
  swing2Motion=Free
  twistLimitDegrees=45
  twistMotion=Free
 }
]
