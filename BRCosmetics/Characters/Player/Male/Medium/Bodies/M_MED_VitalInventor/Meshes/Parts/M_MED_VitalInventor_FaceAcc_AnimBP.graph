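# Animation Blueprint for the VitalInventor face-accessory part, which adds simulated hair motion on top of a pose copied from parent and body meshes.
# Back-hair physics switches between a front-end setup with fixed bone offsets and a gameplay setup that rotates tail and hair bones by head-to-spine delta.
# That cached result feeds a front-hair RigidBody pass.
# All simulations run relative to spine_05, using gravity overrides from a deprecated movement-state function.
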AnimBlueprint /BRCosmetics/Characters/Player/Male/Medium/Bodies/M_MED_VitalInventor/Meshes/Parts/M_MED_VitalInventor_FaceAcc_AnimBP
target_skeleton=&/M_MED_VitalInventor_FaceAcc_Skeleton
graphs [
 AnimGraph {nodes [
   n1 = RigidBody!(
    ComponentPose=Local To Component(LocalPose=Use cached pose 'BackHairCache'())
    bAlphaBoolEnabled:bool
    OverrideWorldGravity:Vector=Get GravityOverride()
    ExternalForce
   ) -> (Pose) {
    alphaInputType=Bool
    bClampLinearTranslationLimitToRefPose=true
    bOverrideWorldGravity=true
    overridePhysicsAsset=&/M_MED_VitalInventor_FaceAcc_FrontHair_Physics
    simulationSpace=BaseBoneSpace
    alphaBoolBlend {blendOption=Linear}
    baseBoneRef {boneName=spine_05}
    componentLinearAccScale=<0.35 0.2 0.2>
    componentLinearVelScale=<0.35 0.2 0.2>
   }
   Output Pose(Result=Component To Local(ComponentPose=@n1)) {layerGroup=None}
   n2 = "Transform (Modify) Bone - Bone: base_tail"(
    ComponentPose=Local To Component(LocalPose=Use cached pose 'StartPose'())
    Alpha=Get AimAndUp() -> (AimAndUp:double)
    Translation="0,0,0"
    Rotation="40,0,0"
    Scale="1,1,1"
   ) {
    rotationMode=BMM_Additive
    rotationSpace=BCS_BoneSpace
    translationMode=BMM_Replace
    translationSpace=BCS_BoneSpace
    alphaBoolBlend {blendOption=Linear}
    boneToModify {boneName=base_tail}
    rotation {pitch=40}
   }
   n3 = "Transform (Modify) Bone - Bone: base_hair_back_r"(ComponentPose=@n2, Alpha=Get Hair Up() -> (Hair Up:double), Translation="0,0,0", Rotation="39.838,0.336,-20.406", Scale="1,1,1") {
    rotationMode=BMM_Additive
    rotationSpace=BCS_BoneSpace
    translationMode=BMM_Replace
    translationSpace=BCS_BoneSpace
    alphaBoolBlend {blendOption=Linear}
    boneToModify {boneName=base_hair_back_r}
    rotation=<rot 39.838 0.336 -20.406>
   }
   n4 = RigidBody!(ComponentPose=@n3, Alpha:float, OverrideWorldGravity=Get GravityOverride(), ExternalForce) -> (Pose) {
    bClampLinearTranslationLimitToRefPose=true
    bOverrideWorldGravity=true
    overridePhysicsAsset=&/M_MED_VitalInventor_FaceAcc_Physics
    simulationSpace=BaseBoneSpace
    alphaBoolBlend {blendOption=Linear}
    baseBoneRef {boneName=spine_05}
    componentLinearAccScale=<0.35 0.3 0.5>
    componentLinearVelScale=<0.35 0.3 0.5>
   }
   n5 = "Transform (Modify) Bone - Bone: base_hair_front_l"(
    ComponentPose=Local To Component(LocalPose=Use cached pose 'StartPose'())
    Alpha=1
    Translation="0,0,0"
    Rotation="5,-0.026,10"
    Scale="1,1,1"
   ) {
    rotationMode=BMM_Additive
    rotationSpace=BCS_BoneSpace
    translationMode=BMM_Replace
    translationSpace=BCS_BoneSpace
    alphaBoolBlend {blendOption=Linear}
    boneToModify {boneName=base_hair_front_l}
    rotation=<rot 5 -0.026 10>
   }
   n6 = "Transform (Modify) Bone - Bone: base_hair_back"(ComponentPose=@n5, Alpha=1, Translation="0,0,0", Rotation="-4.967,-4.335,-0.164", Scale="1,1,1") {
    rotationMode=BMM_Additive
    rotationSpace=BCS_BoneSpace
    translationMode=BMM_Replace
    translationSpace=BCS_BoneSpace
    alphaBoolBlend {blendOption=Linear}
    boneToModify {boneName=base_hair_back}
    rotation=<rot -4.967 -4.335 -0.164>
   }
   n7 = "Transform (Modify) Bone - Bone: dyn_hair_front_01_l"(ComponentPose=@n6, Alpha=1, Translation="0,0,0", Rotation="15,0,35", Scale="1,1,1") {
    rotationMode=BMM_Additive
    rotationSpace=BCS_BoneSpace
    translationMode=BMM_Replace
    translationSpace=BCS_BoneSpace
    alphaBoolBlend {blendOption=Linear}
    boneToModify {boneName=dyn_hair_front_01_l}
    rotation {pitch=15, roll=35}
   }
   n8 = RigidBody(ComponentPose=@n7, OverrideWorldGravity=Get GravityOverride()) {
    bClampLinearTranslationLimitToRefPose=true
    bOverrideWorldGravity=true
    overridePhysicsAsset=&/M_MED_VitalInventor_FaceAcc_FrontEnd_Physics
    simulationSpace=BaseBoneSpace
    alphaBoolBlend {blendOption=Linear}
    baseBoneRef {boneName=spine_05}
    componentLinearAccScale=<0.35 0.3 0.5>
    componentLinearVelScale=<0.35 0.3 0.5>
   }
   Save cached pose 'BackHairCache'(Pose=Blend Poses by bool(BlendPose_0=Component To Local(ComponentPose=@n8), BlendPose_1=Component To Local(ComponentPose=@n4), bActiveValue=Get bIsInFrontEnd()) {blendType=Linear}) {cacheName=BackHairCache}
   n9 = Layered blend per bone(
    BasePose=Copy Pose From Mesh() {bCopyCurves=true, bUseAttachedParent=true}
    BlendPoses_0=Copy Pose From Mesh(SourceMeshComponent=Get Body_Mesh())
   ) {layerSetup [{branchFilters [{boneName=clavicle_l}, {boneName=clavicle_r}, {boneName=upperarm_r}, {boneName=upperarm_l}]}]}
   Save cached pose 'StartPose'(Pose=@n9) {cacheName=StartPose}
  ]}
 EventGraph {nodes [
   Event OnCharacterCustomizationLoaded()
   > n1 = Cast To PlayerPawn_Athena(Object=TryGetPawnOwner())
   > Set Body_Mesh(Body_Mesh=GetSkeletalMeshForPartType(self=@n1))
   n2 = CheckForBackbling()
   > Set isWearingBackbling(isWearingBackbling=AND Boolean!(A=@n2, B=NOT Boolean(A=@n2.bIsBackblingACape), C:bool=NOT Boolean(A=@n2.bIsBackblingLong)) -> (ReturnValue))
   > Set Is Wearing Cape(Is Wearing Cape=OR Boolean(A=@n2.bIsBackblingACape, B=@n2.bIsBackblingLong))
   Event BlueprintInitializeAnimation()
   > Set Skel Mesh(Skel Mesh=GetOwnerSkeletalMeshComponent())
   Event BlueprintUpdateAnimation()
   > n3 = Sequence!(execute) -> (then_0, then_1, then_2:Exec)
   n4 = Branch(execute=@n3.then_2, Condition=Get bRigidBodyEnabled())
   n5 = Get Skel Mesh()
   n6 = Set Hair Up(
    execute=@n3.then_1
    Hair Up:double=float * float(
     A=GetAlphaFromReferenceTransformDelta(SkeletalMeshComponent=@n5, BoneName=spine_05, BaseName=head, InputRange=(X=0,Y=90), TwistAxis="0,0,1")
     B=1
    ) {is=title K2Node_PromotableOperator}
   ) -> (Output_Get:double)
   Set AimAndUp(execute=@n4.then, AimAndUp:double=@n6) -> (Output_Get:double)
   Set AimAndUp(execute=@n4.else, AimAndUp=0)
   n7 = Gravity Override State Function DEPRECATED(
    execute=@n3.then_0
    CharacterPartAnimInstance=Self-Reference()
    GameTimeInSeconds=GetGameTimeInSeconds()
    GravityOverrideParams=Get Gravity Override Params()
   )
   > Set GravityOverride(GravityOverride=@n7)
  ]}
]
variables [
 AimAndUp {type=number}
 Body_Mesh {type=SkeletalMeshComponent}
 Gravity Override Params {
  type=GravityOverrideParamsStruct
  default {
   onGround_Moving {gravityOverride=<0 0 -1000>}
   skydive_FloatingDown {gravityOverride=<0 -700 -800>, windAmplitude=<300 -500 400>, windFrequency=<10 7 8>}
   skydive_GoingUp {gravityOverride=<0 -300 -200>}
  }
 }
 GravityOverride {type=Vector}
 Hair Up {type=number}
 Is Wearing Cape {type=boolean}
 IsCrouching {type=number}
 isWearingBackbling {type=boolean}
 Skel Mesh {type=SkeletalMeshComponent}
]
